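# Physics asset for the medium female bandage ninja backpack cosmetic, defining capsule bodies and joints for the pack and its two dangling bandages.
# Two large capsules cover the pack, with dyn_base rotationally locked to backpack_root.
# Left and right bandages are three-link chains with limited swing and very tight twist, their capsules growing toward the tips.
# Each chain hangs from a tiny master body that rotates freely against backpack_root.
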PhysicsAsset /BRCosmetics/Accessories/FORT_Backpacks/F_MED_Bandage_NInja_Backpack/Meshes/F_MED_Bandage_Ninja_Backpack_Physics
bodies {
 backpack_root [{shapeType=Capsule, length=6.1413, radius=15, shapeName=backpack_root_capsule, center=<-13.0032 10.3584 0>}]
 dyn_base [{shapeType=Capsule, length=9.122, radius=16.1025, shapeName=dyn_base_capsule, center=<-14.02 13.7527 0.0002>}]
 dyn_l_bandage_01 [{shapeType=Capsule, length=2.5085, radius=1.5765, center=<4 0 0>, rotation=<rot 90 -0.0099 -0.0099>}]
 dyn_l_bandage_02 [{shapeType=Capsule, length=4.0467, radius=1.75, center=<5 0 0>, rotation=<rot 90 0 0>}]
 dyn_l_bandage_03 [{shapeType=Capsule, length=6.7087, radius=2.5091, center=<8 0 0>, rotation=<rot 90 -179.9802 -179.9802>}]
 dyn_l_bandage_MSTR [{shapeType=Capsule, length=0.505, radius=0.505, shapeName=dyn_l_bandage_MSTR_capsule, rotation=<rot 0 -90 0>}]
 dyn_r_bandage_01 [{shapeType=Capsule, length=2.2809, radius=1.5765, center=<4 0 0>, rotation=<rot 90 -0.0003 -0.0002>}]
 dyn_r_bandage_02 [{shapeType=Capsule, length=3.4511, radius=1.75, center=<5 0 0>, rotation=<rot 90 -0.0099 -0.0099>}]
 dyn_r_bandage_03 [{shapeType=Capsule, length=6.7087, radius=2.5091, center=<6 0 0>, rotation=<rot 90 0.0099 0.0099>}]
 dyn_r_bandage_MSTR [{shapeType=Capsule, length=0.505, radius=0.505, shapeName=dyn_r_bandage_MSTR_capsule, rotation=<rot 0 -90 0>}]
}
constraints [
 {bone1Name=dyn_l_bandage_01, bone2Name=dyn_l_bandage_MSTR, swing1LimitDegrees=30, swing2LimitDegrees=15, twistLimitDegrees=2}
 {bone1Name=dyn_l_bandage_02, bone2Name=dyn_l_bandage_01, swing1LimitDegrees=45, swing2LimitDegrees=30, twistLimitDegrees=2}
 {bone1Name=dyn_l_bandage_03, bone2Name=dyn_l_bandage_02, swing1LimitDegrees=45, swing2LimitDegrees=30, twistLimitDegrees=2}
 {bone1Name=dyn_r_bandage_01, bone2Name=dyn_r_bandage_MSTR, swing1LimitDegrees=30, swing2LimitDegrees=15, twistLimitDegrees=2}
 {bone1Name=dyn_r_bandage_02, bone2Name=dyn_r_bandage_01, swing1LimitDegrees=45, swing2LimitDegrees=30, twistLimitDegrees=2}
 {bone1Name=dyn_r_bandage_03, bone2Name=dyn_r_bandage_02, swing1LimitDegrees=45, swing2LimitDegrees=30, twistLimitDegrees=2}
 {
  bone1Name=dyn_base
  bone2Name=backpack_root
  swing1LimitDegrees=45
  swing1Motion=Locked
  swing2LimitDegrees=45
  swing2Motion=Locked
  twistLimitDegrees=45
  twistMotion=Locked
 }
 {
  bone1Name=dyn_r_bandage_MSTR
  bone2Name=backpack_root
  swing1LimitDegrees=45
  swing1Motion=Free
  swing2LimitDegrees=45
  swing2Motion=Free
  twistLimitDegrees=45
  twistMotion=Free
 }
 {
  bone1Name=dyn_l_bandage_MSTR
  bone2Name=backpack_root
  swing1LimitDegrees=45
  swing1Motion=Free
  swing2LimitDegrees=45
  swing2Motion=Free
  twistLimitDegrees=45
  twistMotion=Free
 }
]
