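# Animation Blueprint for the Bandage Ninja backpack cosmetic, adding secondary motion to its dangling bandages and posture correction to the pack.
# Spine bend drives additive bone offsets, feeding a cached pose and final root adjustment.
# A bool blends between rigid-body simulation and bandage Trail controllers on that pose.
# Player movement scales trail alpha, zeroed when rising fast while skydiving; skydiving with a closed parachute lowers simulation gravity.
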AnimBlueprint /BRCosmetics/Accessories/FORT_Backpacks/F_MED_Bandage_NInja_Backpack/Meshes/F_MED_Bandage_Ninja_Backpack_AnimBlueprint
target_skeleton=&/F_MED_Bandage_Ninja_Backpack_Skeleton
graphs [
 AnimGraph {nodes [
   n1 = RigidBody!(
    ComponentPose=Local To Component(LocalPose=Use cached pose 'perPose'())
    Alpha
    OverrideWorldGravity:Vector=Get gravity_override() -> (gravity_override:Vector)
    ExternalForce
   ) -> (Pose) {
    bClampLinearTranslationLimitToRefPose=true
    bOverrideWorldGravity=true
    overridePhysicsAsset=&/F_MED_Bandage_Ninja_Backpack_Physics
    simulationSpace=BaseBoneSpace
    alphaBoolBlend {blendOption=Linear}
    baseBoneRef {boneName=backpack_root}
    componentAppliedLinearAccClamp=<3000 3000 3000>
    componentLinearAccScale=<0.15 0.1>
    componentLinearVelScale=<0.15 0.1>
    overrideWorldGravity {z=-980}
    simSpaceSettings {maxAngularAcceleration=20, maxAngularVelocity=2, maxLinearAcceleration=100, maxLinearVelocity=10, velocityScaleZ=0, worldAlpha=1}
   }
   n2 = Component To Local(ComponentPose=@n1)
   n3 = Get BwdMovement() -> (BwdMovement:double)
   n4 = Trail controller - Bone: dyn_r_bandage_04(ComponentPose=Local To Component(LocalPose=Use cached pose 'perPose'()), Alpha=@n3) {
    chainLength=3
    debugLifeTime=0.1
    maxDeltaTime=0.333
    alphaScaleBiasClamp {bInterpResult=true}
    rotationLimits [{limitMax=<180 180 180>}, {limitMax=<180 180 0>}, {limitMax=<180 180 0>}]
    rotationOffsets [<0 0 0>, <0 0 0>, <0 0 0>]
    trailBone {boneName=dyn_r_bandage_04}
    trailRelaxationSpeed {editorCurveData {keys [
       {arriveTangent=-26.8666, interpMode=RCIM_Linear, time=0, value=18}
       {arriveTangent=-13.886, interpMode=RCIM_Linear, leaveTangent=-13.886, time=0.5, value=15}
       <at 1 5 RCIM_Linear>
      ]}}
   }
   n5 = Trail controller - Bone: dyn_l_bandage_04(ComponentPose=@n4, Alpha=@n3) {
    chainLength=3
    debugLifeTime=0.1
    maxDeltaTime=0.333
    alphaScaleBiasClamp {bInterpResult=true}
    rotationLimits [{limitMax=<0 30 15>, limitMin=<0 -30 -15>}, {limitMax=<180 180 180>}, {limitMax=<180 180 180>}]
    rotationOffsets [<0 0 0>, <0 0 0>, <0 0 0>]
    trailBone {boneName=dyn_l_bandage_04}
    trailRelaxationSpeed {editorCurveData {keys [
       {arriveTangent=-26.8666, interpMode=RCIM_Linear, time=0, value=18}
       {arriveTangent=-13.886, interpMode=RCIM_Linear, leaveTangent=-13.886, time=0.5, value=15}
       <at 1 5 RCIM_Linear>
      ]}}
   }
   n6 = Component To Local(ComponentPose=@n5)
   n7 = Get bRigidBodyEnabled()
   n8 = "Transform (Modify) Bone - Bone: backpack_root"(
    ComponentPose=Local To Component(LocalPose=Blend Poses by bool(BlendPose_0=@n2, BlendPose_1=@n6, bActiveValue=@n7) {blendType=Linear})
    Alpha=1
    Translation=Get backpack_translation()
    Rotation=Get backpack_rotate()
    Scale="1,1,1"
   ) {rotationMode=BMM_Additive, translationMode=BMM_Additive, alphaBoolBlend {blendOption=Linear}, boneToModify {boneName=backpack_root}}
   Output Pose(Result=Component To Local(ComponentPose=@n8)) {layerGroup=None}
   n9 = "Transform (Modify) Bone - Bone: backpack_root"(
    ComponentPose=Local To Component(LocalPose=Input Pose())
    Alpha=Get rot_back_alpha() -> (rot_back_alpha:double)
    Translation="0,0,0"
    Rotation="0.000,10,0.000"
    Scale="1,1,1"
   ) {
    rotationMode=BMM_Additive
    rotationSpace=BCS_BoneSpace
    translationMode=BMM_Additive
    translationSpace=BCS_BoneSpace
    alphaBoolBlend {blendOption=Linear}
    boneToModify {boneName=backpack_root}
    rotation {yaw=10}
   }
   n10 = "Transform (Modify) Bone - Bone: dyn_base"(ComponentPose=@n9, Alpha=Get rot_fwd_alpha() -> (rot_fwd_alpha:double), Translation="0.000,3,0", Rotation="0,0,0", Scale="1,1,1") {
    rotationMode=BMM_Additive
    rotationSpace=BCS_BoneSpace
    translationMode=BMM_Additive
    translationSpace=BCS_BoneSpace
    alphaBoolBlend {blendOption=Linear}
    alphaScaleBiasClamp {bInterpResult=true}
    boneToModify {boneName=dyn_base}
    translation {y=3}
   }
   n11 = Get rot_back_alpha()
   n12 = "Transform (Modify) Bone - Bone: dyn_r_bandage_MSTR"(ComponentPose=@n10, Alpha=@n11, Translation="0,0,0", Rotation="0.000,-35,0.000", Scale="1,1,1") {
    rotationMode=BMM_Additive
    rotationSpace=BCS_BoneSpace
    alphaBoolBlend {blendOption=Linear}
    boneToModify {boneName=dyn_r_bandage_MSTR}
    rotation {yaw=-35}
   }
   n13 = "Transform (Modify) Bone - Bone: dyn_l_bandage_MSTR"(ComponentPose=@n12, Alpha=@n11, Translation="0,0,0", Rotation="0.000,-35,0.000", Scale="1,1,1") {
    rotationMode=BMM_Additive
    rotationSpace=BCS_BoneSpace
    alphaBoolBlend {blendOption=Linear}
    boneToModify {boneName=dyn_l_bandage_MSTR}
    rotation {yaw=-35}
   }
   n14 = "Transform (Modify) Bone - Bone: dyn_base"(ComponentPose=@n13, Alpha=@n11, Translation="0.000,-5.5,-0.000", Rotation="0,0,0", Scale="1,1,1") {
    rotationMode=BMM_Additive
    rotationSpace=BCS_BoneSpace
    translationMode=BMM_Additive
    translationSpace=BCS_BoneSpace
    alphaBoolBlend {blendOption=Linear}
    alphaScaleBiasClamp {bInterpResult=true}
    boneToModify {boneName=dyn_base}
    translation {y=-5.5}
   }
   n15 = "Transform (Modify) Bone - Bone: dyn_base"(ComponentPose=@n14, Alpha=Get twist_alpha() -> (twist_alpha:double), Translation="0.000,-4,-0.000", Rotation="0,0,0", Scale="1,1,1") {
    rotationMode=BMM_Additive
    rotationSpace=BCS_BoneSpace
    translationMode=BMM_Additive
    translationSpace=BCS_BoneSpace
    alphaBoolBlend {blendOption=Linear}
    alphaScaleBiasClamp {bInterpResult=true}
    boneToModify {boneName=dyn_base}
    translation {y=-4}
   }
   Save cached pose 'perPose'(Pose=Component To Local(ComponentPose=@n15)) {cacheName=perPose}
  ]}
 EventGraph {nodes [
   Event BlueprintUpdateAnimation()
   > n1 = Sequence!(execute) -> (then_0, then_1, then_2:Exec, then_3:Exec, then_4:Exec)
   > n2 = Branch(Condition=Get bIsSkyDiving())
   > n3 = Branch(Condition=Get bIsParachuteOpened())
   Set gravity_override(execute=[@n3.then, @n2.else], gravity_override="0, 0,-580")
   Set gravity_override(execute=@n3.else, gravity_override="0, 0,-200")
   n4 = GetOwnerSkeletalMeshComponent()
   n5 = BreakRotator(InRot=BreakTransform(InTransform=GetDeltaTransformFromRefPose(self=@n4, BoneName=spine_05, BaseName=spine_02)).Rotation)
   Set rot_back_alpha(execute=@n1.then_1, rot_back_alpha:double=MapRangeClamped(Value=@n5.Yaw, InRangeA=0, InRangeB=-15)) -> (Output_Get:double)
   > Set rot_fwd_alpha(rot_fwd_alpha:double=MapRangeClamped(Value=@n5.Yaw, InRangeA=0, InRangeB=30)) -> (Output_Get:double)
   > Set twist_alpha(twist_alpha:double=MapRangeClamped(Value="Absolute (Float)"(A=@n5), InRangeA=0, InRangeB=30)) -> (Output_Get:double)
   n6 = PlayerMovement(execute=@n1.then_2, AnimInstance=Self-Reference(), XDotProductCoefficient=1, YDotProductCoefficient=1, ZDotProductCoefficient=1)
   n7 = MapRangeClamped(Value=@n6.YMovement, InRangeA=0, InRangeB=-1, OutRangeA=1, OutRangeB=0)
   n8 = MapRangeClamped(Value="Absolute (Float)"(A=@n6), InRangeA=0, InRangeB=1, OutRangeA=1, OutRangeB=0.4)
   > Set BwdMovement(BwdMovement:double=MapRangeClamped(Value=float * float(A=@n8, B=@n7), InRangeA=0, InRangeB=1.5, OutRangeB=0.8)) -> (Output_Get:double)
   n9 = Branch(execute=@n1.then_3, Condition=Get bIsSkyDiving())
   > n10 = Branch(Condition="float > float"(A=Get PawnVelocityZ(), B=50) {is=title K2Node_CallFunction})
   > Set BwdMovement(BwdMovement=float * float(A=Get BwdMovement(), B=0))
   Set BwdMovement(execute=[@n9.else, @n10.else], BwdMovement=float * float(A=Get BwdMovement(), B=1))
   n11 = GetOwnerSkeletalMeshComponent()
   n12 = BreakRotator(InRot=BreakTransform(InTransform=GetDeltaTransformFromRefPose(self=@n11, BoneName=spine_05, BaseName=spine_01)).Rotation)
   Set backpack_rotate(execute=@n1.then_4, backpack_rotate=MakeRotator(Roll=MapRangeClamped(Value=@n12.Yaw, InRangeA=0, InRangeB=-30, OutRangeB=0)))
   > Set backpack_translation(backpack_translation=MakeVector(Y=MapRangeClamped(Value=@n12.Yaw, InRangeA=0, InRangeB=-30, OutRangeB=-5)))
   n13 = GetOwnerSkeletalMeshComponent()
   n14 = BreakRotator(InRot=BreakTransform(InTransform=GetDeltaTransformFromRefPose(self=@n13, BoneName=spine_05, BaseName=spine_02)).Rotation)
   n15 = MapRangeClamped(Value=@n14.Yaw, InRangeA=0, InRangeB=-30, OutRangeB=15)
   Set backpack_rotate(backpack_rotate=MakeRotator(Yaw=@n15))
   > Set backpack_translation(backpack_translation=MakeVector(Y=MapRangeClamped(Value=@n14.Yaw, InRangeA=0, InRangeB=-30, OutRangeB=-2.5)))
   n16 = MapRangeClamped(Value=@n14.Yaw, InRangeA=0, InRangeB=12, OutRangeB=-25)
   n17 = MapRangeClamped(Value=@n14.Yaw, InRangeA=0, InRangeB=-30, OutRangeB=10)
   > Set base_rotate(base_rotate:Rotator=MakeRotator(Yaw=float + float(A=@n16, B=@n17) {is=title K2Node_CommutativeAssociativeBinaryOperator})) -> (Output_Get:Rotator)
   "Absolute (Float)"(A=@n5.Yaw)
   float + float(A=@n15, B=MapRangeClamped(Value=@n14.Yaw, InRangeA=0, InRangeB=12, OutRangeB=-5))
  ]}
]
variables [
 backpack_rotate {type=Rotator}
 backpack_translation {type=Vector}
 base_rotate {type=Rotator}
 BwdMovement {type=number}
 gravity_override {type=Vector}
 rot_back_alpha {type=number}
 rot_fwd_alpha {type=number}
 twist_alpha {type=number}
]
